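# UI-domain material for batch action widget font in a tech-art editor utility, building an animated hue-cycling color gradient.
# Time, wrapped by a period and scaled by a Speed parameter, scrolls the texture coordinates.
# A Rotator, angled by a scalar parameter rather than time, turns those coordinates and drives the HueShift function's percentage on the ColorBase color, so hue varies across UV space.
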Material /Game/Blueprints/Blutilities/ForTechArt/AssetTools/BatchAction/UIMaterial/BatchActionWidgetFontMaterial
graph {
 properties {materialDomain=MD_UI, pixelDepthOffsetMode=PDOM_Legacy, usedShadingModels=Default Lit}
 nodes [
  n1 = TextureCoordinate()
  n2 = Multiply(A=Time() {bOverride_Period=true, period=600}, B=ScalarParameter() {defaultValue=-0.2, parameterName=Speed})
  MaterialFunctionCall(
   Hue Shift Percentage=Rotator(Coordinate=Add(A=@n1, B=@n2), Time=ScalarParameter() {defaultValue=-3.14, parameterName=Rotator})
   Texture=VectorParameter() {parameterName=ColorBase, defaultValue=<rgba 1 0.5 0.5 1>}.RGB
  ) {materialFunction=&HueShift}
 ]
}
