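# Editor utility widget for tech art that batch renames assets and optionally moves them into another asset's folder.
# Tick rebuilds old and new paths from the edited arrays, clearing duplicate assets and flagging invalid, clashing or existing targets with placeholder text.
# Hovering the action button previews path changes; clicking runs RenameAssets, which calls Python EditorAssetLibrary.rename_asset for valid, changed paths.
# Property views bound to the widget itself provide the editable lists, and Reset clears them.
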EditorUtilityWidgetBlueprint /Game/Blueprints/Blutilities/ForTechArt/AssetTools/BatchAction/BatchRenameAndMoveAssetTool
graphs [
 EventGraph {nodes [
   "On Clicked (Button_141)"() {is=title K2Node_ComponentBoundEvent}
   n1 = Get CanvasPanel_2() -> (CanvasPanel_2:Canvas Panel Object Reference)
   > SetVisibility(self=@n1, InVisibility=Visible)
   > Delay!(execute, WorldContextObject:Object Reference, Duration=1) -> (then)
   > RenameAssets()
   > Delay(Duration=1)
   > SetVisibility(self=@n1, InVisibility=Collapsed)
   "On Hovered (Button_141)"() {is=title K2Node_ComponentBoundEvent}
   n2 = Get ConfirmView() -> (ConfirmView:Canvas Panel Object Reference)
   > SetVisibility(self=@n2, InVisibility=HitTestInvisible)
   "On Unhovered (Button_141)"() {is=title K2Node_ComponentBoundEvent}
   > SetVisibility(self=@n2, InVisibility=Collapsed)
   "On Clicked (Button)"() {is=title K2Node_ComponentBoundEvent}
   > Clear(TargetArray:Array of Object References=Get ReadyRenameAssets() -> (ReadyRenameAssets:Array of Object References))
   > Clear(TargetArray=Get MoveToThisObjectFolder() -> (MoveToThisObjectFolder:Array of Object References))
   > Clear(TargetArray:Array of Strings=Get OldNames() -> (OldNames:Array of Strings))
   > Clear(TargetArray=Get NewNames() -> (NewNames:Array of Strings))
   > Clear(TargetArray=Get OldPath() -> (OldPath:Array of Strings))
   > Clear(TargetArray=Get NewPath() -> (NewPath:Array of Strings))
   Event Construct()
   > SetObject(
    self=[
     Get ToFolder() -> (ToFolder:Fort Multi Property View Object Reference)
     Get ConfirmRenameAssets() -> (ConfirmRenameAssets:Fort Multi Property View Object Reference)
     Get OldPathV() -> (OldPathV:Fort Multi Property View Object Reference)
     Get NetPathV() -> (NetPathV:Fort Multi Property View Object Reference)
     Get OldName() -> (OldName:Fort Multi Property View Object Reference)
     Get NewName() -> (NewName:Fort Multi Property View Object Reference)
     Get ObjInput() -> (ObjInput:Fort Multi Property View Object Reference)
    ]
    NewObject=Self-Reference()
   )
   Event Tick()
   n3 = Get ReadyRenameAssets()
   > n4 = For Each Loop(Array:Array of Object References=@n3) -> (Array Element:Object Reference)
   Clear(execute=@n4.Completed, TargetArray=Get OldPath())
   n5 = Get OldNames()
   > Clear(TargetArray=@n5)
   > n6 = For Each Loop(Array=Get ReadyRenameAssets())
   n7 = Sequence(execute=@n6.Completed)
   > n8 = Branch(Condition=IsChecked(self=Get CheckBox_38() -> (CheckBox_38:Check Box Object Reference)))
   n9 = Get ToFolder()
   > SetVisibility(self=@n9, InVisibility=Visible)
   SetVisibility(execute=@n8.else, self=@n9, InVisibility=Collapsed)
   n10 = Get NewNames()
   n11 = For Each Loop(Exec=@n7.then_1, Array:Array of Strings=@n10) -> (Array Element:String)
   > n12 = IsThisNameAlreadyUse!(execute, self, Name:String=@n11) -> (then, AlreadyUse:bool)
   > Branch(Condition=@n12)
   > Set Array Elem!(execute, TargetArray=@n10, Index:int=@n11.Array Index, Item:"String (by ref)"=ThisWillNotExecute*, bSizeToFit:bool=true) -> (then)
   n13 = Sequence(execute=@n6.LoopBody)
   > Add(TargetArray=@n5, NewItem=GetObjectName(Object=@n6)) {is=title K2Node_CallArrayFunction}
   > n14 = Sequence()
   n15 = Get NewNames()
   > n16 = Branch(Condition=Is Valid Index!(TargetArray=@n15, IndexToTest:int=@n6.Array Index) -> (ReturnValue:bool))
   > Set Array Elem(
    TargetArray=@n15
    Index=@n6.Array Index
    Item="Get (a ref)"(Array:Array of Strings=@n15, Dimension 1=@n6.Array Index) -> (Output:"String (by ref)")
    bSizeToFit=true
   )
   Set Array Elem(execute=@n16.else, TargetArray=@n15, Index=@n6.Array Index, Item=ThisWillNotExecute*, bSizeToFit=true)
   n17 = Get MoveToThisObjectFolder()
   n18 = Branch(execute=@n14.then_1, Condition=Is Valid Index(TargetArray:Array of Object References=@n17, IndexToTest=@n6.Array Index))
   n19 = "Get (a copy)"(Array:Array of Object References=@n17, Dimension 1=@n6.Array Index) -> (Output:Object Reference)
   > n20 = Is Valid(InputObject=@n19)
   Set Array Elem(
    execute=[@n22.then, @n18.else, @n20.Is Not Valid]
    TargetArray:Array of Object References=@n17
    Index=@n6.Array Index
    Item:"Object Reference (by ref)"=@n6
    bSizeToFit=true
   )
   n21 = Get ReadyRenameAssets()
   n22 = Branch(
    execute=@n20.Is Valid
    Condition="integer >= integer"(A=Find Item!(TargetArray:Array of Object References=@n21, ItemToFind:"Object Reference (by ref)"=@n19) -> (ReturnValue))
   )
   Set Array Elem(execute=@n22.else, TargetArray=@n17, Index=@n6.Array Index, Item=@n19, bSizeToFit=true)
   n23 = GetPathNameForLoadedAsset(execute=@n13.then_1, LoadedAsset=@n6)
   n24 = Split!(InPath:String=@n23) -> (PathPart:String, FilenamePart:String, ExtensionPart:String)
   > Add(TargetArray=Get OldPath(), NewItem=Append!(A=@n24, B=/, C:String=@n24.FilenamePart) -> (ReturnValue))
   > n25 = Is Valid(InputObject=@n6)
   n26 = Get NewPath()
   Set Array Elem(
    execute=[@n41.else, @n34.else, @n25.Is Not Valid]
    TargetArray:Array of Strings=@n26
    Index=@n6.Array Index
    Item:"String (by ref)"=ThisWillNotExecute*
    bSizeToFit=true
   )
   n27 = Branch(execute=@n25.Is Valid, Condition=IsChecked(self=Get CheckBox_38()))
   n28 = "Get (a copy)"(Array=Get MoveToThisObjectFolder(), Dimension 1=@n6.Array Index)
   > n29 = Is Valid(InputObject=@n28)
   n30 = GetPathNameForLoadedAsset(execute=[@n27.else, @n29.Is Not Valid], LoadedAsset=@n6)
   n31 = Split(InPath=@n30)
   n32 = "Get (a copy)"(Array:Array of Strings=Get NewNames(), Dimension 1=@n6.Array Index) -> (Output:String)
   n33 = Append(A=@n31, B=/, C=@n32)
   > n34 = Branch(Condition=ValidatePath!(InPath:String=@n33) -> (bDidSucceed:bool, OutReason:Text))
   n35 = GetProjectContentDirectory!() -> (ReturnValue:String)
   n36 = Split!(SourceString:String=@n31, InStr:String=/Game, SearchCase:ESearchCase Enum=CaseSensitive, SearchDir:ESearchDir Enum=FromStart) -> (LeftS:String, RightS:String, ReturnValue:bool)
   > n37 = Branch(Condition=FileExists!(InPath:String=Append!(A=@n35, B=@n36.RightS, C=/, D:String=@n32, E:String=.uasset) -> (ReturnValue)) -> (ReturnValue:bool))
   Set Array Elem(execute=[@n37.then, @n44.then], TargetArray=@n26, Index=@n6.Array Index, Item=FileExists*, bSizeToFit=true)
   Set Array Elem(execute=@n37.else, TargetArray=@n26, Index=@n6.Array Index, Item=@n33, bSizeToFit=true)
   n38 = GetPathNameForLoadedAsset(execute=@n29.Is Valid, LoadedAsset=@n28)
   n39 = Split!(InPath:String=@n38) -> (PathPart:String, FilenamePart:String, ExtensionPart:String)
   n40 = Append!(A=@n39, B=/, C=@n32) -> (ReturnValue)
   > n41 = Branch(Condition=ValidatePath(InPath=@n40))
   n42 = GetProjectContentDirectory()
   n43 = Split!(SourceString:String=@n39, InStr:String=/Game, SearchCase:ESearchCase Enum=CaseSensitive, SearchDir:ESearchDir Enum=FromStart) -> (LeftS:String, RightS:String, ReturnValue:bool)
   > n44 = Branch(Condition=FileExists(InPath=Append!(A=@n42, B=@n43.RightS, C=/, D:String=@n32, E:String=.uasset) -> (ReturnValue)))
   Set Array Elem(execute=@n44.else, TargetArray=@n26, Index=@n6.Array Index, Item=@n40, bSizeToFit=true)
   Is Valid(exec=@n4.LoopBody, InputObject=@n4)
   > n45 = For Each Loop(Array:Array of Object References=@n3) -> (Array Element:Object Reference)
   > Branch(Condition=AND Boolean(A="Not Equal (Integer)"(A=@n45.Array Index, B=@n4.Array Index), B="Equal (Object)"(A=@n4, B=@n45)))
   > Set Array Elem(TargetArray:Array of Object References=@n3, Index=@n45.Array Index, Item:"Object Reference (by ref)", bSizeToFit=true)
   Event PreConstruct()
  ]}
 IsThisNameAlreadyUse {nodes [
   n1 = IsThisNameAlreadyUse!() -> (then, Name:String) {is=title K2Node_FunctionEntry}
   > n2 = For Each Loop(Array:Array of Strings=Get NewNames()) -> (Array Element:String)
   n3 = Get Indexs() -> (Indexs:Array of Integers)
   Return Node!(execute=@n2.Completed, AlreadyUse:bool="integer > integer"(A=Length(TargetArray:Array of Integers=@n3), B=1))
   > Branch(Condition="Equal Exactly (String)"(A=@n2, B=@n1))
   > Add(TargetArray:Array of Integers=Get Indexs(), NewItem:"Integer (by ref)"=@n2.Array Index)
  ]}
 RenameAssets {nodes [
   RenameAssets() {is=title K2Node_FunctionEntry}
   > n1 = For Each Loop(Array:Array of Object References=Get ReadyRenameAssets()) -> (Array Element:Object Reference)
   > Is Valid(InputObject=@n1)
   n2 = "Get (a copy)"(Array=Get NewPath(), Dimension 1=@n1.Array Index)
   n3 = Get OldPath()
   > n4 = Branch(Condition="Equal Exactly (String)"(A="Get (a copy)"(Array=@n3, Dimension 1=@n1.Array Index), B=@n2))
   Branch(execute=@n4.else, Condition=ValidatePath(InPath=@n2))
   > n5 = GetPathNameForLoadedAsset(LoadedAsset=@n1)
   > Execute Python Script(PythonScript=Append!(A="import unreal\r\n", B="unreal.EditorAssetLibrary.rename_asset(\"", C=@n5, D="\",\"", E=@n2, F:String="\")") -> (ReturnValue))
   n6 = Split(SourceString=@n2, InStr=/, SearchDir=FromEnd)
   Make Asset Rename Data!(Asset:Object Weak object=@n1, NewPackagePath:String=@n6, NewName:String=@n6.RightS) -> (AssetRenameData:Asset Rename Data Structure) {is=title K2Node_MakeStruct}
  ]}
]
variables [
 MoveToThisObjectFolder {type="array<Object>", default []}
 NewNames {type="array<string>", default []}
 NewPath {type="array<string>", default []}
 OldNames {type="array<string>", default []}
 OldPath {type="array<string>", default []}
 ReadyRenameAssets {type="array<Object>", default []}
]
widgets {
 info {rootWidgetClass=&CanvasPanel, widgetCount=35}
 widgets [
  CanvasPanel_0 {widgetClassPath=&CanvasPanel, widget=CanvasPanel_0}
  CanvasPanel_1 {
   widgetClassPath=&CanvasPanel
   parent=CanvasPanel_0
   widget=CanvasPanel_1
   slot {class=CanvasPanelSlot, properties {layoutData {alignment=<0 0>, anchors {maximum=<1 1>, minimum=<0 0>}, offsets=<ltrb 5 5 5 5>}}}
  }
  Overlay_1 {
   widgetClassPath=&Overlay
   parent=CanvasPanel_1
   widget=Overlay_1
   slot {class=CanvasPanelSlot, properties {layoutData {alignment=<0 0>, anchors {maximum=<1 1>, minimum=<0 0>}, offsets=<ltrb 0 0 0 0>}}}
  }
  VerticalBox_0 {
   widgetClassPath=&VerticalBox
   parent=Overlay_1
   widget=VerticalBox_0
   slot {class=OverlaySlot, properties {horizontalAlignment=HAlign_Fill, verticalAlignment=VAlign_Fill}}
  }
  Overlay_2 {
   widgetClassPath=&Overlay
   parent=VerticalBox_0
   widget=Overlay_2
   slot {class=VerticalBoxSlot, properties {verticalAlignment=VAlign_Center}}
  }
  UEFN_TextBlock_C_88 {
   widgetClassPath=&UEFN_TextBlock
   parent=Overlay_2
   widget=UEFN_TextBlock_C_88
   properties {
    text=RenameTool
    colorAndOpacity {colorUseRule=UseColor_Specified, specifiedColor=<rgba 0 0.0458 0.05 1>}
    font {
     bForceMonospaced=false
     bMaterialIsStencil=false
     fontMaterial=/Game/Blueprints/Blutilities/ForTechArt/AssetTools/BatchAction/UIMaterial/BatchActionWidgetFontMaterial_Inst
     fontObject=&BurbankSmall
     letterSpacing=0
     monospacedWidth=1
     size=24
     skewAmount=0
     typefaceFontName=Black
     outlineSettings {
      bApplyOutlineToDropShadows=false
      bMiteredCorners=false
      bSeparateFillAlpha=false
      outlineMaterial=/Game/Blueprints/Blutilities/ForTechArt/AssetTools/BatchAction/UIMaterial/BatchActionWidgetFontMaterial
      outlineSize=1
      outlineColor=<rgba 0 1 0.5 1>
     }
    }
   }
   slot {class=OverlaySlot, properties {horizontalAlignment=HAlign_Center, verticalAlignment=VAlign_Center}}
  }
  CheckBox_38 {
   widgetClassPath=/Script/UMG.CheckBox
   bIsVariable=true
   parent=Overlay_2
   widget=CheckBox_38
   properties {widgetStyle {
     checkBoxType=CheckBox
     backgroundHoveredImage {
      bIsDynamicallyLoaded=false
      drawAs=RoundedBox
      imageType=NoImage
      mirroring=NoMirror
      resourceName=None
      resourceObject=null
      tiling=NoTile
      imageSize=<18 18>
      margin=<ltrb 0 0 0 0>
      outlineSettings {
       bUseBrushTransparency=false
       roundingType=FixedRadius
       width=1
       color {colorUseRule=UseColor_ColorTable, specifiedColor=<rgba 0 0 0 0>}
       cornerRadii=<3 3 3 3>
      }
      tintColor {colorUseRule=UseColor_Specified, specifiedColor=<rgba 0.0048 0.0048 0.0048 1>}
      uVRegion {bIsValid=false, max=<0 0>, min=<0 0>}
     }
     backgroundImage {
      bIsDynamicallyLoaded=false
      drawAs=RoundedBox
      imageType=NoImage
      mirroring=NoMirror
      resourceName=None
      resourceObject=null
      tiling=NoTile
      imageSize=<18 18>
      margin=<ltrb 0 0 0 0>
      outlineSettings {
       bUseBrushTransparency=false
       roundingType=FixedRadius
       width=1
       color {colorUseRule=UseColor_ColorTable, specifiedColor=<rgba 0 0 0 0>}
       cornerRadii=<3 3 3 3>
      }
      tintColor {colorUseRule=UseColor_Specified, specifiedColor=<rgba 0.0048 0.0048 0.0048 1>}
      uVRegion {bIsValid=false, max=<0 0>, min=<0 0>}
     }
     backgroundPressedImage {
      bIsDynamicallyLoaded=false
      drawAs=RoundedBox
      imageType=NoImage
      mirroring=NoMirror
      resourceName=None
      resourceObject=null
      tiling=NoTile
      imageSize=<18 18>
      margin=<ltrb 0 0 0 0>
      outlineSettings {
       bUseBrushTransparency=false
       roundingType=FixedRadius
       width=1
       color {colorUseRule=UseColor_ColorTable, specifiedColor=<rgba 0 0 0 0>}
       cornerRadii=<3 3 3 3>
      }
      tintColor {colorUseRule=UseColor_Specified, specifiedColor=<rgba 0.0048 0.0048 0.0048 1>}
      uVRegion {bIsValid=false, max=<0 0>, min=<0 0>}
     }
     borderBackgroundColor {colorUseRule=UseColor_Specified, specifiedColor=<rgba 1 1 1 1>}
     checkedForeground {colorUseRule=UseColor_Specified, specifiedColor=<rgba 1 1 1 1>}
     checkedHoveredForeground {colorUseRule=UseColor_Specified, specifiedColor=<rgba 1 1 1 1>}
     checkedHoveredImage {
      bIsDynamicallyLoaded=false
      drawAs=Image
      imageType=Vector
      mirroring=NoMirror
      resourceName=../../../Engine/Content/Slate/Starship/CoreWidgets/CheckBox/check.svg
      resourceObject=null
      tiling=NoTile
      imageSize=<16 16>
      margin=<ltrb 0 0 0 0>
      outlineSettings {
       bUseBrushTransparency=false
       roundingType=HalfHeightRadius
       width=0
       color {colorUseRule=UseColor_Specified, specifiedColor=<rgba 0 0 0 0>}
       cornerRadii=<0 0 0 0>
      }
      tintColor {colorUseRule=UseColor_Specified, specifiedColor=<rgba 0.0044 0.2384 1 1>}
      uVRegion {bIsValid=false, max=<0 0>, min=<0 0>}
     }
     checkedImage {
      bIsDynamicallyLoaded=false
      drawAs=Image
      imageType=Vector
      mirroring=NoMirror
      resourceName=../../../Engine/Content/Slate/Starship/CoreWidgets/CheckBox/check.svg
      resourceObject=null
      tiling=NoTile
      imageSize=<16 16>
      margin=<ltrb 0 0 0 0>
      outlineSettings {
       bUseBrushTransparency=false
       roundingType=HalfHeightRadius
       width=0
       color {colorUseRule=UseColor_Specified, specifiedColor=<rgba 0 0 0 0>}
       cornerRadii=<0 0 0 0>
      }
      tintColor {colorUseRule=UseColor_Specified, specifiedColor=<rgba 0 0.162 0.7454 1>}
      uVRegion {bIsValid=false, max=<0 0>, min=<0 0>}
     }
     checkedPressedForeground {colorUseRule=UseColor_Specified, specifiedColor=<rgba 1 1 1 1>}
     checkedPressedImage {
      bIsDynamicallyLoaded=false
      drawAs=Image
      imageType=Vector
      mirroring=NoMirror
      resourceName=../../../Engine/Content/Slate/Starship/CoreWidgets/CheckBox/check.svg
      resourceObject=null
      tiling=NoTile
      imageSize=<16 16>
      margin=<ltrb 0 0 0 0>
      outlineSettings {
       bUseBrushTransparency=false
       roundingType=HalfHeightRadius
       width=0
       color {colorUseRule=UseColor_Specified, specifiedColor=<rgba 0 0 0 0>}
       cornerRadii=<0 0 0 0>
      }
      tintColor {colorUseRule=UseColor_Specified, specifiedColor=<rgba 0 0.0802 0.3515 1>}
      uVRegion {bIsValid=false, max=<0 0>, min=<0 0>}
     }
     checkedSlateSound {resourceObject=null}
     foregroundColor {colorUseRule=UseColor_Specified, specifiedColor=<rgba 1 1 1 1>}
     hoveredForeground {colorUseRule=UseColor_Specified, specifiedColor=<rgba 1 1 1 1>}
     hoveredSlateSound {resourceObject=null}
     padding=<ltrb 2 0 0 0>
     pressedForeground {colorUseRule=UseColor_Specified, specifiedColor=<rgba 1 1 1 1>}
     uncheckedHoveredImage {
      bIsDynamicallyLoaded=false
      drawAs=NoDrawType
      imageType=NoImage
      mirroring=NoMirror
      resourceName=None
      resourceObject=null
      tiling=NoTile
      imageSize=<0 0>
      margin=<ltrb 0 0 0 0>
      outlineSettings {
       bUseBrushTransparency=false
       roundingType=HalfHeightRadius
       width=0
       color {colorUseRule=UseColor_Specified, specifiedColor=<rgba 0 0 0 0>}
       cornerRadii=<0 0 0 0>
      }
      tintColor {colorUseRule=UseColor_Specified, specifiedColor=<rgba 1 1 1 1>}
      uVRegion {bIsValid=false, max=<0 0>, min=<0 0>}
     }
     uncheckedImage {
      bIsDynamicallyLoaded=false
      drawAs=NoDrawType
      imageType=NoImage
      mirroring=NoMirror
      resourceName=None
      resourceObject=null
      tiling=NoTile
      imageSize=<0 0>
      margin=<ltrb 0 0 0 0>
      outlineSettings {
       bUseBrushTransparency=false
       roundingType=HalfHeightRadius
       width=0
       color {colorUseRule=UseColor_Specified, specifiedColor=<rgba 0 0 0 0>}
       cornerRadii=<0 0 0 0>
      }
      tintColor {colorUseRule=UseColor_Specified, specifiedColor=<rgba 1 1 1 1>}
      uVRegion {bIsValid=false, max=<0 0>, min=<0 0>}
     }
     uncheckedPressedImage {
      bIsDynamicallyLoaded=false
      drawAs=NoDrawType
      imageType=NoImage
      mirroring=NoMirror
      resourceName=None
      resourceObject=null
      tiling=NoTile
      imageSize=<0 0>
      margin=<ltrb 0 0 0 0>
      outlineSettings {
       bUseBrushTransparency=false
       roundingType=HalfHeightRadius
       width=0
       color {colorUseRule=UseColor_Specified, specifiedColor=<rgba 0 0 0 0>}
       cornerRadii=<0 0 0 0>
      }
      tintColor {colorUseRule=UseColor_Specified, specifiedColor=<rgba 1 1 1 1>}
      uVRegion {bIsValid=false, max=<0 0>, min=<0 0>}
     }
     uncheckedSlateSound {resourceObject=null}
     undeterminedForeground {colorUseRule=UseColor_Specified, specifiedColor=<rgba 1 1 1 1>}
     undeterminedHoveredImage {
      bIsDynamicallyLoaded=false
      drawAs=Image
      imageType=Vector
      mirroring=NoMirror
      resourceName=../../../Engine/Content/Slate/Starship/CoreWidgets/CheckBox/indeterminate.svg
      resourceObject=null
      tiling=NoTile
      imageSize=<16 16>
      margin=<ltrb 0 0 0 0>
      outlineSettings {
       bUseBrushTransparency=false
       roundingType=HalfHeightRadius
       width=0
       color {colorUseRule=UseColor_Specified, specifiedColor=<rgba 0 0 0 0>}
       cornerRadii=<0 0 0 0>
      }
      tintColor {colorUseRule=UseColor_Specified, specifiedColor=<rgba 0.0044 0.2384 1 1>}
      uVRegion {bIsValid=false, max=<0 0>, min=<0 0>}
     }
     undeterminedImage {
      bIsDynamicallyLoaded=false
      drawAs=Image
      imageType=Vector
      mirroring=NoMirror
      resourceName=../../../Engine/Content/Slate/Starship/CoreWidgets/CheckBox/indeterminate.svg
      resourceObject=null
      tiling=NoTile
      imageSize=<16 16>
      margin=<ltrb 0 0 0 0>
      outlineSettings {
       bUseBrushTransparency=false
       roundingType=HalfHeightRadius
       width=0
       color {colorUseRule=UseColor_Specified, specifiedColor=<rgba 0 0 0 0>}
       cornerRadii=<0 0 0 0>
      }
      tintColor {colorUseRule=UseColor_Specified, specifiedColor=<rgba 0 0.162 0.7454 1>}
      uVRegion {bIsValid=false, max=<0 0>, min=<0 0>}
     }
     undeterminedPressedImage {
      bIsDynamicallyLoaded=false
      drawAs=Image
      imageType=Vector
      mirroring=NoMirror
      resourceName=../../../Engine/Content/Slate/Starship/CoreWidgets/CheckBox/indeterminate.svg
      resourceObject=null
      tiling=NoTile
      imageSize=<16 16>
      margin=<ltrb 0 0 0 0>
      outlineSettings {
       bUseBrushTransparency=false
       roundingType=HalfHeightRadius
       width=0
       color {colorUseRule=UseColor_Specified, specifiedColor=<rgba 0 0 0 0>}
       cornerRadii=<0 0 0 0>
      }
      tintColor {colorUseRule=UseColor_Specified, specifiedColor=<rgba 0 0.0802 0.3515 1>}
      uVRegion {bIsValid=false, max=<0 0>, min=<0 0>}
     }
    }}
   slot {class=OverlaySlot, properties {horizontalAlignment=HAlign_Right, verticalAlignment=VAlign_Fill}}
  }
  UEFN_TextBlock_C_2 {
   widgetClassPath=&UEFN_TextBlock
   parent=CheckBox_38
   widget=UEFN_TextBlock_C_2
   properties {
    text=Change destination folder
    colorAndOpacity {colorUseRule=UseColor_Specified, specifiedColor=<rgba 0.2 0.2 0.2 1>}
    font {
     bForceMonospaced=false
     bMaterialIsStencil=false
     fontMaterial=null
     fontObject=&BurbankSmall
     letterSpacing=0
     monospacedWidth=1
     size=16
     skewAmount=0
     typefaceFontName=Bold
     outlineSettings {
      bApplyOutlineToDropShadows=false
      bMiteredCorners=false
      bSeparateFillAlpha=false
      outlineMaterial=null
      outlineSize=0
      outlineColor=<rgba 0 0 0 1>
     }
    }
   }
   slot {class=PanelSlot}
  }
  Border_1 {
   widgetClassPath=&Border
   parent=VerticalBox_0
   widget=Border_1
   properties {
    background {
     bIsDynamicallyLoaded=false
     drawAs=Image
     imageType=NoImage
     mirroring=NoMirror
     resourceName=None
     resourceObject=&BatchActionWidgetFontMaterial
     tiling=NoTile
     imageSize=<32 32>
     margin=<ltrb 0 0 0 0>
     outlineSettings {
      bUseBrushTransparency=false
      roundingType=HalfHeightRadius
      width=0
      color {colorUseRule=UseColor_Specified, specifiedColor=<rgba 0 0 0 0>}
      cornerRadii=<0 0 0 0>
     }
     tintColor {colorUseRule=UseColor_Specified, specifiedColor=<rgba 0 1 0.5 1>}
     uVRegion {bIsValid=false, max=<0 0>, min=<0 0>}
    }
    padding=<ltrb 1 1 1 1>
   }
   slot {class=VerticalBoxSlot, properties {size {sizeRule=Fill, value=1}}}
  }
  Overlay_0 {widgetClassPath=&Overlay, parent=Border_1, widget=Overlay_0, slot {class=BorderSlot, properties {padding=<ltrb 1 1 1 1>}}}
  Border_37 {
   widgetClassPath=&Border
   parent=Overlay_0
   widget=Border_37
   properties {background {
     bIsDynamicallyLoaded=false
     drawAs=Image
     imageType=NoImage
     mirroring=NoMirror
     resourceName=None
     resourceObject=null
     tiling=NoTile
     imageSize=<32 32>
     margin=<ltrb 0 0 0 0>
     outlineSettings {
      bUseBrushTransparency=false
      roundingType=HalfHeightRadius
      width=0
      color {colorUseRule=UseColor_Specified, specifiedColor=<rgba 0 0 0 0>}
      cornerRadii=<0 0 0 0>
     }
     tintColor {colorUseRule=UseColor_Specified, specifiedColor=<rgba 0.0194 0.0194 0.0194 1>}
     uVRegion {bIsValid=false, max=<0 0>, min=<0 0>}
    }}
   slot {class=OverlaySlot, properties {horizontalAlignment=HAlign_Fill, verticalAlignment=VAlign_Fill}}
  }
  ScrollBox_28 {
   widgetClassPath=&ScrollBox
   parent=Overlay_0
   widget=ScrollBox_28
   properties {
    widgetBarStyle {
     thickness=8
     draggedThumbImage {
      bIsDynamicallyLoaded=false
      drawAs=RoundedBox
      imageType=NoImage
      mirroring=NoMirror
      resourceName=None
      resourceObject=null
      tiling=NoTile
      imageSize=<0 0>
      margin=<ltrb 0 0 0 0>
      outlineSettings {
       bUseBrushTransparency=false
       roundingType=FixedRadius
       width=0
       color {colorUseRule=UseColor_Specified, specifiedColor=<rgba 0 0 0 0>}
       cornerRadii=<4 4 4 4>
      }
      tintColor {colorUseRule=UseColor_Specified, specifiedColor=<rgba 0.2159 0.2159 0.2159 1>}
      uVRegion {bIsValid=false, max=<0 0>, min=<0 0>}
     }
     horizontalBackgroundImage {
      bIsDynamicallyLoaded=false
      drawAs=NoDrawType
      imageType=NoImage
      mirroring=NoMirror
      resourceName=None
      resourceObject=null
      tiling=NoTile
      imageSize=<0 0>
      margin=<ltrb 0 0 0 0>
      outlineSettings {
       bUseBrushTransparency=false
       roundingType=HalfHeightRadius
       width=0
       color {colorUseRule=UseColor_Specified, specifiedColor=<rgba 0 0 0 0>}
       cornerRadii=<0 0 0 0>
      }
      tintColor {colorUseRule=UseColor_Specified, specifiedColor=<rgba 1 1 1 1>}
      uVRegion {bIsValid=false, max=<0 0>, min=<0 0>}
     }
     horizontalBottomSlotImage {
      bIsDynamicallyLoaded=false
      drawAs=NoDrawType
      imageType=NoImage
      mirroring=NoMirror
      resourceName=None
      resourceObject=null
      tiling=NoTile
      imageSize=<0 0>
      margin=<ltrb 0 0 0 0>
      outlineSettings {
       bUseBrushTransparency=false
       roundingType=HalfHeightRadius
       width=0
       color {colorUseRule=UseColor_Specified, specifiedColor=<rgba 0 0 0 0>}
       cornerRadii=<0 0 0 0>
      }
      tintColor {colorUseRule=UseColor_Specified, specifiedColor=<rgba 1 1 1 1>}
      uVRegion {bIsValid=false, max=<0 0>, min=<0 0>}
     }
     horizontalTopSlotImage {
      bIsDynamicallyLoaded=false
      drawAs=NoDrawType
      imageType=NoImage
      mirroring=NoMirror
      resourceName=None
      resourceObject=null
      tiling=NoTile
      imageSize=<0 0>
      margin=<ltrb 0 0 0 0>
      outlineSettings {
       bUseBrushTransparency=false
       roundingType=HalfHeightRadius
       width=0
       color {colorUseRule=UseColor_Specified, specifiedColor=<rgba 0 0 0 0>}
       cornerRadii=<0 0 0 0>
      }
      tintColor {colorUseRule=UseColor_Specified, specifiedColor=<rgba 1 1 1 1>}
      uVRegion {bIsValid=false, max=<0 0>, min=<0 0>}
     }
     hoveredThumbImage {
      bIsDynamicallyLoaded=false
      drawAs=RoundedBox
      imageType=NoImage
      mirroring=NoMirror
      resourceName=None
      resourceObject=null
      tiling=NoTile
      imageSize=<0 0>
      margin=<ltrb 0 0 0 0>
      outlineSettings {
       bUseBrushTransparency=false
       roundingType=FixedRadius
       width=0
       color {colorUseRule=UseColor_Specified, specifiedColor=<rgba 0 0 0 0>}
       cornerRadii=<4 4 4 4>
      }
      tintColor {colorUseRule=UseColor_Specified, specifiedColor=<rgba 0.2159 0.2159 0.2159 1>}
      uVRegion {bIsValid=false, max=<0 0>, min=<0 0>}
     }
     normalThumbImage {
      bIsDynamicallyLoaded=false
      drawAs=RoundedBox
      imageType=NoImage
      mirroring=NoMirror
      resourceName=None
      resourceObject=null
      tiling=NoTile
      imageSize=<0 0>
      margin=<ltrb 0 0 0 0>
      outlineSettings {
       bUseBrushTransparency=false
       roundingType=FixedRadius
       width=0
       color {colorUseRule=UseColor_Specified, specifiedColor=<rgba 0 0 0 0>}
       cornerRadii=<4 4 4 4>
      }
      tintColor {colorUseRule=UseColor_Specified, specifiedColor=<rgba 0.0953 0.0953 0.0953 1>}
      uVRegion {bIsValid=false, max=<0 0>, min=<0 0>}
     }
     verticalBackgroundImage {
      bIsDynamicallyLoaded=false
      drawAs=NoDrawType
      imageType=NoImage
      mirroring=NoMirror
      resourceName=None
      resourceObject=null
      tiling=NoTile
      imageSize=<0 0>
      margin=<ltrb 0 0 0 0>
      outlineSettings {
       bUseBrushTransparency=false
       roundingType=HalfHeightRadius
       width=0
       color {colorUseRule=UseColor_Specified, specifiedColor=<rgba 0 0 0 0>}
       cornerRadii=<0 0 0 0>
      }
      tintColor {colorUseRule=UseColor_Specified, specifiedColor=<rgba 1 1 1 1>}
      uVRegion {bIsValid=false, max=<0 0>, min=<0 0>}
     }
     verticalBottomSlotImage {
      bIsDynamicallyLoaded=false
      drawAs=NoDrawType
      imageType=NoImage
      mirroring=NoMirror
      resourceName=None
      resourceObject=null
      tiling=NoTile
      imageSize=<0 0>
      margin=<ltrb 0 0 0 0>
      outlineSettings {
       bUseBrushTransparency=false
       roundingType=HalfHeightRadius
       width=0
       color {colorUseRule=UseColor_Specified, specifiedColor=<rgba 0 0 0 0>}
       cornerRadii=<0 0 0 0>
      }
      tintColor {colorUseRule=UseColor_Specified, specifiedColor=<rgba 1 1 1 1>}
      uVRegion {bIsValid=false, max=<0 0>, min=<0 0>}
     }
     verticalTopSlotImage {
      bIsDynamicallyLoaded=false
      drawAs=NoDrawType
      imageType=NoImage
      mirroring=NoMirror
      resourceName=None
      resourceObject=null
      tiling=NoTile
      imageSize=<0 0>
      margin=<ltrb 0 0 0 0>
      outlineSettings {
       bUseBrushTransparency=false
       roundingType=HalfHeightRadius
       width=0
       color {colorUseRule=UseColor_Specified, specifiedColor=<rgba 0 0 0 0>}
       cornerRadii=<0 0 0 0>
      }
      tintColor {colorUseRule=UseColor_Specified, specifiedColor=<rgba 1 1 1 1>}
      uVRegion {bIsValid=false, max=<0 0>, min=<0 0>}
     }
    }
    widgetStyle {
     barThickness=8
     bottomShadowBrush {
      bIsDynamicallyLoaded=false
      drawAs=Box
      imageType=FullColor
      mirroring=NoMirror
      resourceName=../../../Engine/Content/Slate/Common/ScrollBoxShadowBottom.png
      resourceObject=null
      tiling=NoTile
      imageSize=<16 8>
      margin=<ltrb 0.5 0 0.5 1>
      outlineSettings {
       bUseBrushTransparency=false
       roundingType=HalfHeightRadius
       width=0
       color {colorUseRule=UseColor_Specified, specifiedColor=<rgba 0 0 0 0>}
       cornerRadii=<0 0 0 0>
      }
      tintColor {colorUseRule=UseColor_Specified, specifiedColor=<rgba 1 1 1 1>}
      uVRegion {bIsValid=false, max=<0 0>, min=<0 0>}
     }
     horizontalScrolledContentPadding=<ltrb 0 0 1 0>
     leftShadowBrush {
      bIsDynamicallyLoaded=false
      drawAs=Box
      imageType=FullColor
      mirroring=NoMirror
      resourceName=../../../Engine/Content/Slate/Common/ScrollBoxShadowLeft.png
      resourceObject=null
      tiling=NoTile
      imageSize=<8 16>
      margin=<ltrb 1 0.5 0 0.5>
      outlineSettings {
       bUseBrushTransparency=false
       roundingType=HalfHeightRadius
       width=0
       color {colorUseRule=UseColor_Specified, specifiedColor=<rgba 0 0 0 0>}
       cornerRadii=<0 0 0 0>
      }
      tintColor {colorUseRule=UseColor_Specified, specifiedColor=<rgba 1 1 1 1>}
      uVRegion {bIsValid=false, max=<0 0>, min=<0 0>}
     }
     rightShadowBrush {
      bIsDynamicallyLoaded=false
      drawAs=Box
      imageType=FullColor
      mirroring=NoMirror
      resourceName=../../../Engine/Content/Slate/Common/ScrollBoxShadowRight.png
      resourceObject=null
      tiling=NoTile
      imageSize=<8 16>
      margin=<ltrb 0 0.5 1 0.5>
      outlineSettings {
       bUseBrushTransparency=false
       roundingType=HalfHeightRadius
       width=0
       color {colorUseRule=UseColor_Specified, specifiedColor=<rgba 0 0 0 0>}
       cornerRadii=<0 0 0 0>
      }
      tintColor {colorUseRule=UseColor_Specified, specifiedColor=<rgba 1 1 1 1>}
      uVRegion {bIsValid=false, max=<0 0>, min=<0 0>}
     }
     topShadowBrush {
      bIsDynamicallyLoaded=false
      drawAs=Box
      imageType=FullColor
      mirroring=NoMirror
      resourceName=../../../Engine/Content/Slate/Common/ScrollBoxShadowTop.png
      resourceObject=null
      tiling=NoTile
      imageSize=<16 8>
      margin=<ltrb 0.5 1 0.5 0>
      outlineSettings {
       bUseBrushTransparency=false
       roundingType=HalfHeightRadius
       width=0
       color {colorUseRule=UseColor_Specified, specifiedColor=<rgba 0 0 0 0>}
       cornerRadii=<0 0 0 0>
      }
      tintColor {colorUseRule=UseColor_Specified, specifiedColor=<rgba 1 1 1 1>}
      uVRegion {bIsValid=false, max=<0 0>, min=<0 0>}
     }
     verticalScrolledContentPadding=<ltrb 0 0 0 1>
    }
   }
   slot {class=OverlaySlot, properties {horizontalAlignment=HAlign_Fill, verticalAlignment=VAlign_Fill, padding=<ltrb 5 5 5 5>}}
  }
  HorizontalBox_85 {widgetClassPath=&HorizontalBox, parent=ScrollBox_28, widget=HorizontalBox_85, slot {class=ScrollBoxSlot}}
  ObjInput {
   widgetClassPath=/Script/FortniteEditor.FortMultiPropertyView
   bIsVariable=true
   parent=HorizontalBox_85
   widget=ObjInput
   properties {propertiesToShow [ReadyRenameAssets]}
   slot {class=HorizontalBoxSlot, properties {size {sizeRule=Fill, value=1.25}}}
  }
  ToFolder {
   widgetClassPath=&FortMultiPropertyView
   bIsVariable=true
   parent=HorizontalBox_85
   widget=ToFolder
   properties {propertiesToShow [MoveToThisObjectFolder]}
   slot {class=HorizontalBoxSlot, properties {size {sizeRule=Fill, value=1.25}}}
  }
  OldName {
   widgetClassPath=&FortMultiPropertyView
   bIsVariable=true
   parent=HorizontalBox_85
   widget=OldName
   properties {propertiesToShow [OldNames]}
   slot {class=HorizontalBoxSlot, properties {size {sizeRule=Fill, value=1}}}
  }
  NewName {
   widgetClassPath=&FortMultiPropertyView
   bIsVariable=true
   parent=HorizontalBox_85
   widget=NewName
   properties {propertiesToShow [NewNames]}
   slot {class=HorizontalBoxSlot, properties {size {sizeRule=Fill, value=1}}}
  }
  ConfirmView {
   widgetClassPath=&CanvasPanel
   bIsVariable=true
   parent=Overlay_0
   widget=ConfirmView
   properties {visibility=Collapsed}
   slot {class=OverlaySlot, properties {horizontalAlignment=HAlign_Fill, verticalAlignment=VAlign_Fill, padding=<ltrb 5 5 5 5>}}
  }
  Border_216 {
   widgetClassPath=&Border
   parent=ConfirmView
   widget=Border_216
   properties {
    background {
     bIsDynamicallyLoaded=false
     drawAs=Image
     imageType=NoImage
     mirroring=NoMirror
     resourceName=None
     resourceObject=null
     tiling=NoTile
     imageSize=<32 32>
     margin=<ltrb 0 0 0 0>
     outlineSettings {
      bUseBrushTransparency=false
      roundingType=HalfHeightRadius
      width=0
      color {colorUseRule=UseColor_Specified, specifiedColor=<rgba 0 0 0 0>}
      cornerRadii=<0 0 0 0>
     }
     tintColor {colorUseRule=UseColor_Specified, specifiedColor=<rgba 0.04 0.04 0.04 1>}
     uVRegion {bIsValid=false, max=<0 0>, min=<0 0>}
    }
    padding=<ltrb 0 0 0 0>
   }
   slot {class=CanvasPanelSlot, properties {layoutData {alignment=<0 0>, anchors {maximum=<1 1>, minimum=<0 0>}, offsets=<ltrb 0 0 0 0>}}}
  }
  VerticalBox_2 {
   widgetClassPath=&VerticalBox
   parent=Border_216
   widget=VerticalBox_2
   slot {class=BorderSlot, properties {padding=<ltrb 0 0 0 0>}}
  }
  HorizontalBox_88 {
   widgetClassPath=&HorizontalBox
   parent=VerticalBox_2
   widget=HorizontalBox_88
   slot {class=VerticalBoxSlot, properties {size {sizeRule=Fill, value=1}}}
  }
  ConfirmRenameAssets {
   widgetClassPath=&FortMultiPropertyView
   bIsVariable=true
   parent=HorizontalBox_88
   widget=ConfirmRenameAssets
   properties {propertiesToShow [ReadyRenameAssets]}
   slot {class=HorizontalBoxSlot, properties {size {sizeRule=Fill, value=0.75}}}
  }
  OldPathV {
   widgetClassPath=&FortMultiPropertyView
   bIsVariable=true
   parent=HorizontalBox_88
   widget=OldPathV
   properties {propertiesToShow [OldPath]}
   slot {class=HorizontalBoxSlot, properties {size {sizeRule=Fill, value=1}}}
  }
  NetPathV {
   widgetClassPath=&FortMultiPropertyView
   bIsVariable=true
   parent=HorizontalBox_88
   widget=NetPathV
   properties {propertiesToShow [NewPath]}
   slot {class=HorizontalBoxSlot, properties {size {sizeRule=Fill, value=1}}}
  }
  UEFN_TextBlock_C_4 {
   widgetClassPath=&UEFN_TextBlock
   parent=VerticalBox_2
   widget=UEFN_TextBlock_C_4
   properties {
    justification=Center
    text="If you have many changes in assets,please CheckIN before rename!!!!!!!!!!!!!!!!!!"
    colorAndOpacity {colorUseRule=UseColor_Specified, specifiedColor=<rgba 1 0 0 1>}
    font {
     bForceMonospaced=false
     bMaterialIsStencil=false
     fontMaterial=null
     fontObject=&BurbankSmall
     letterSpacing=0
     monospacedWidth=1
     size=16
     skewAmount=0
     typefaceFontName=Bold
     outlineSettings {
      bApplyOutlineToDropShadows=false
      bMiteredCorners=false
      bSeparateFillAlpha=false
      outlineMaterial=null
      outlineSize=0
      outlineColor=<rgba 0 0 0 1>
     }
    }
   }
   slot {class=VerticalBoxSlot, properties {horizontalAlignment=HAlign_Center, verticalAlignment=VAlign_Center}}
  }
  Button {
   widgetClassPath=&Button
   parent=VerticalBox_0
   widget=Button
   properties {widgetStyle {
     clickedSlateSound {resourceObject=null}
     disabled {
      bIsDynamicallyLoaded=false
      drawAs=RoundedBox
      imageType=NoImage
      mirroring=NoMirror
      resourceName=None
      resourceObject=null
      tiling=NoTile
      imageSize=<0 0>
      margin=<ltrb 0 0 0 0>
      outlineSettings {
       bUseBrushTransparency=false
       roundingType=FixedRadius
       width=1
       color {colorUseRule=UseColor_ColorTable, specifiedColor=<rgba 0 0 0 0>}
       cornerRadii=<4 4 4 4>
      }
      tintColor {colorUseRule=UseColor_Specified, specifiedColor=<rgba 0.0395 0.0395 0.0395 1>}
      uVRegion {bIsValid=false, max=<0 0>, min=<0 0>}
     }
     disabledForeground {colorUseRule=UseColor_ColorTable, specifiedColor=<rgba 0 0 0 0>}
     hovered {
      bIsDynamicallyLoaded=false
      drawAs=RoundedBox
      imageType=NoImage
      mirroring=NoMirror
      resourceName=None
      resourceObject=null
      tiling=NoTile
      imageSize=<32 32>
      margin=<ltrb 0 0 0 0>
      outlineSettings {
       bUseBrushTransparency=false
       roundingType=FixedRadius
       width=1
       color {colorUseRule=UseColor_ColorTable, specifiedColor=<rgba 0 0 0 0>}
       cornerRadii=<4 4 4 4>
      }
      tintColor {colorUseRule=UseColor_Specified, specifiedColor=<rgba 0.0953 0.0953 0.0953 1>}
      uVRegion {bIsValid=false, max=<0 0>, min=<0 0>}
     }
     hoveredForeground {colorUseRule=UseColor_ColorTable, specifiedColor=<rgba 0 0 0 0>}
     hoveredSlateSound {resourceObject=null}
     normal {
      bIsDynamicallyLoaded=false
      drawAs=RoundedBox
      imageType=NoImage
      mirroring=NoMirror
      resourceName=None
      resourceObject=null
      tiling=NoTile
      imageSize=<32 32>
      margin=<ltrb 0 0 0 0>
      outlineSettings {
       bUseBrushTransparency=false
       roundingType=FixedRadius
       width=1
       color {colorUseRule=UseColor_ColorTable, specifiedColor=<rgba 0 0 0 0>}
       cornerRadii=<4 4 4 4>
      }
      tintColor {colorUseRule=UseColor_Specified, specifiedColor=<rgba 0.0395 0.0395 0.0395 1>}
      uVRegion {bIsValid=false, max=<0 0>, min=<0 0>}
     }
     normalForeground {colorUseRule=UseColor_ColorTable, specifiedColor=<rgba 0 0 0 0>}
     normalPadding=<ltrb 12 1.5 12 1.5>
     pressed {
      bIsDynamicallyLoaded=false
      drawAs=RoundedBox
      imageType=NoImage
      mirroring=NoMirror
      resourceName=None
      resourceObject=null
      tiling=NoTile
      imageSize=<32 32>
      margin=<ltrb 0 0 0 0>
      outlineSettings {
       bUseBrushTransparency=false
       roundingType=FixedRadius
       width=1
       color {colorUseRule=UseColor_ColorTable, specifiedColor=<rgba 0 0 0 0>}
       cornerRadii=<4 4 4 4>
      }
      tintColor {colorUseRule=UseColor_Specified, specifiedColor=<rgba 0.0284 0.0284 0.0284 1>}
      uVRegion {bIsValid=false, max=<0 0>, min=<0 0>}
     }
     pressedForeground {colorUseRule=UseColor_ColorTable, specifiedColor=<rgba 0 0 0 0>}
     pressedPadding=<ltrb 12 2.5 12 0.5>
     pressedSlateSound {resourceObject=null}
    }}
   slot {class=VerticalBoxSlot, properties {padding=<ltrb 0 5 0 0>}}
  }
  UEFN_TextBlock_C_3 {
   widgetClassPath=&UEFN_TextBlock
   parent=Button
   widget=UEFN_TextBlock_C_3
   properties {
    text=Reset
    colorAndOpacity {colorUseRule=UseColor_Specified, specifiedColor=<rgba 0.05 0 0.0132 1>}
    font {
     bForceMonospaced=false
     bMaterialIsStencil=false
     fontMaterial=&BatchActionWidgetFontMaterial_Inst
     fontObject=&BurbankSmall
     letterSpacing=0
     monospacedWidth=1
     size=20
     skewAmount=0
     typefaceFontName=Black
     outlineSettings {
      bApplyOutlineToDropShadows=false
      bMiteredCorners=false
      bSeparateFillAlpha=false
      outlineMaterial=&BatchActionWidgetFontMaterial
      outlineSize=1
      outlineColor=<rgba 1 0 0.1278 1>
     }
    }
   }
   slot {class=ButtonSlot, properties {verticalAlignment=VAlign_Fill, padding=<ltrb 0 0 0 0>}}
  }
  Button_141 {
   widgetClassPath=&Button
   parent=VerticalBox_0
   widget=Button_141
   properties {widgetStyle {
     clickedSlateSound {resourceObject=null}
     disabled {
      bIsDynamicallyLoaded=false
      drawAs=RoundedBox
      imageType=NoImage
      mirroring=NoMirror
      resourceName=None
      resourceObject=null
      tiling=NoTile
      imageSize=<0 0>
      margin=<ltrb 0 0 0 0>
      outlineSettings {
       bUseBrushTransparency=false
       roundingType=FixedRadius
       width=1
       color {colorUseRule=UseColor_ColorTable, specifiedColor=<rgba 0 0 0 0>}
       cornerRadii=<4 4 4 4>
      }
      tintColor {colorUseRule=UseColor_Specified, specifiedColor=<rgba 0.0395 0.0395 0.0395 1>}
      uVRegion {bIsValid=false, max=<0 0>, min=<0 0>}
     }
     disabledForeground {colorUseRule=UseColor_ColorTable, specifiedColor=<rgba 0 0 0 0>}
     hovered {
      bIsDynamicallyLoaded=false
      drawAs=RoundedBox
      imageType=NoImage
      mirroring=NoMirror
      resourceName=None
      resourceObject=null
      tiling=NoTile
      imageSize=<32 32>
      margin=<ltrb 0 0 0 0>
      outlineSettings {
       bUseBrushTransparency=false
       roundingType=FixedRadius
       width=1
       color {colorUseRule=UseColor_ColorTable, specifiedColor=<rgba 0 0 0 0>}
       cornerRadii=<4 4 4 4>
      }
      tintColor {colorUseRule=UseColor_Specified, specifiedColor=<rgba 0.0953 0.0953 0.0953 1>}
      uVRegion {bIsValid=false, max=<0 0>, min=<0 0>}
     }
     hoveredForeground {colorUseRule=UseColor_ColorTable, specifiedColor=<rgba 0 0 0 0>}
     hoveredSlateSound {resourceObject=null}
     normal {
      bIsDynamicallyLoaded=false
      drawAs=RoundedBox
      imageType=NoImage
      mirroring=NoMirror
      resourceName=None
      resourceObject=null
      tiling=NoTile
      imageSize=<32 32>
      margin=<ltrb 0 0 0 0>
      outlineSettings {
       bUseBrushTransparency=false
       roundingType=FixedRadius
       width=1
       color {colorUseRule=UseColor_ColorTable, specifiedColor=<rgba 0 0 0 0>}
       cornerRadii=<4 4 4 4>
      }
      tintColor {colorUseRule=UseColor_Specified, specifiedColor=<rgba 0.0395 0.0395 0.0395 1>}
      uVRegion {bIsValid=false, max=<0 0>, min=<0 0>}
     }
     normalForeground {colorUseRule=UseColor_ColorTable, specifiedColor=<rgba 0 0 0 0>}
     normalPadding=<ltrb 12 1.5 12 1.5>
     pressed {
      bIsDynamicallyLoaded=false
      drawAs=RoundedBox
      imageType=NoImage
      mirroring=NoMirror
      resourceName=None
      resourceObject=null
      tiling=NoTile
      imageSize=<32 32>
      margin=<ltrb 0 0 0 0>
      outlineSettings {
       bUseBrushTransparency=false
       roundingType=FixedRadius
       width=1
       color {colorUseRule=UseColor_ColorTable, specifiedColor=<rgba 0 0 0 0>}
       cornerRadii=<4 4 4 4>
      }
      tintColor {colorUseRule=UseColor_Specified, specifiedColor=<rgba 0.0284 0.0284 0.0284 1>}
      uVRegion {bIsValid=false, max=<0 0>, min=<0 0>}
     }
     pressedForeground {colorUseRule=UseColor_ColorTable, specifiedColor=<rgba 0 0 0 0>}
     pressedPadding=<ltrb 12 2.5 12 0.5>
     pressedSlateSound {resourceObject=null}
    }}
   slot {class=VerticalBoxSlot, properties {padding=<ltrb 0 5 0 0>}}
  }
  UEFN_TextBlock_C {
   widgetClassPath=&UEFN_TextBlock
   parent=Button_141
   widget=UEFN_TextBlock_C
   properties {
    text=Execute
    colorAndOpacity {colorUseRule=UseColor_Specified, specifiedColor=<rgba 0 0.0458 0.05 1>}
    font {
     bForceMonospaced=false
     bMaterialIsStencil=false
     fontMaterial=&BatchActionWidgetFontMaterial_Inst
     fontObject=&BurbankSmall
     letterSpacing=0
     monospacedWidth=1
     size=20
     skewAmount=0
     typefaceFontName=Black
     outlineSettings {
      bApplyOutlineToDropShadows=false
      bMiteredCorners=false
      bSeparateFillAlpha=false
      outlineMaterial=&BatchActionWidgetFontMaterial
      outlineSize=1
      outlineColor=<rgba 0 1 0.5 1>
     }
    }
   }
   slot {class=ButtonSlot, properties {verticalAlignment=VAlign_Fill, padding=<ltrb 0 0 0 0>}}
  }
  CanvasPanel_2 {
   widgetClassPath=&CanvasPanel
   bIsVariable=true
   parent=Overlay_1
   widget=CanvasPanel_2
   properties {visibility=Collapsed}
   slot {class=OverlaySlot, properties {horizontalAlignment=HAlign_Fill, verticalAlignment=VAlign_Fill}}
  }
  Border_97 {
   widgetClassPath=&Border
   parent=CanvasPanel_2
   widget=Border_97
   properties {
    background {
     bIsDynamicallyLoaded=false
     drawAs=Image
     imageType=NoImage
     mirroring=NoMirror
     resourceName=None
     resourceObject=&BatchActionWidgetFontMaterial
     tiling=NoTile
     imageSize=<32 32>
     margin=<ltrb 0 0 0 0>
     outlineSettings {
      bUseBrushTransparency=false
      roundingType=HalfHeightRadius
      width=0
      color {colorUseRule=UseColor_Specified, specifiedColor=<rgba 0 0 0 0>}
      cornerRadii=<0 0 0 0>
     }
     tintColor {colorUseRule=UseColor_Specified, specifiedColor=<rgba 0 1 0.5 1>}
     uVRegion {bIsValid=false, max=<0 0>, min=<0 0>}
    }
    padding=<ltrb 1 1 1 1>
   }
   slot {class=CanvasPanelSlot, properties {layoutData {alignment=<0 0>, anchors {maximum=<1 1>, minimum=<0 0>}, offsets=<ltrb 0 0 0 0>}}}
  }
  Border_164 {
   widgetClassPath=&Border
   parent=Border_97
   widget=Border_164
   properties {
    horizontalAlignment=HAlign_Center
    verticalAlignment=VAlign_Center
    background {
     bIsDynamicallyLoaded=false
     drawAs=Image
     imageType=NoImage
     mirroring=NoMirror
     resourceName=None
     resourceObject=null
     tiling=NoTile
     imageSize=<32 32>
     margin=<ltrb 0 0 0 0>
     outlineSettings {
      bUseBrushTransparency=false
      roundingType=HalfHeightRadius
      width=0
      color {colorUseRule=UseColor_Specified, specifiedColor=<rgba 0 0 0 0>}
      cornerRadii=<0 0 0 0>
     }
     tintColor {colorUseRule=UseColor_Specified, specifiedColor=<rgba 0.0194 0.0194 0.0194 1>}
     uVRegion {bIsValid=false, max=<0 0>, min=<0 0>}
    }
    padding=<ltrb 0 0 0 0>
   }
   slot {class=BorderSlot, properties {padding=<ltrb 1 1 1 1>}}
  }
  VerticalBox_1 {
   widgetClassPath=&VerticalBox
   parent=Border_164
   widget=VerticalBox_1
   slot {class=BorderSlot, properties {horizontalAlignment=HAlign_Center, verticalAlignment=VAlign_Center, padding=<ltrb 0 0 0 0>}}
  }
  UEFN_TextBlock_C_130 {
   widgetClassPath=&UEFN_TextBlock
   parent=VerticalBox_1
   widget=UEFN_TextBlock_C_130
   properties {
    justification=Center
    text=Processing...
    colorAndOpacity {colorUseRule=UseColor_Specified, specifiedColor=<rgba 0 0.0458 0.05 1>}
    font {
     bForceMonospaced=false
     bMaterialIsStencil=false
     fontMaterial=&BatchActionWidgetFontMaterial_Inst
     fontObject=&BurbankSmall
     letterSpacing=0
     monospacedWidth=1
     size=24
     skewAmount=0
     typefaceFontName=Black
     outlineSettings {
      bApplyOutlineToDropShadows=false
      bMiteredCorners=false
      bSeparateFillAlpha=false
      outlineMaterial=&BatchActionWidgetFontMaterial
      outlineSize=1
      outlineColor=<rgba 0 1 0.5 1>
     }
    }
   }
   slot {class=VerticalBoxSlot}
  }
  UEFN_TextBlock_C_1 {
   widgetClassPath=&UEFN_TextBlock
   parent=VerticalBox_1
   widget=UEFN_TextBlock_C_1
   properties {
    justification=Center
    text="Each file takes 10-60 seconds according to the network speed. \r\nPlease wait patiently and do not close the editor.\r\nTake a coffee."
    colorAndOpacity {colorUseRule=UseColor_Specified, specifiedColor=<rgba 1 0.02 0.1 1>}
    font {
     bForceMonospaced=false
     bMaterialIsStencil=false
     fontMaterial=&BatchActionWidgetFontMaterial
     fontObject=&BurbankSmall
     letterSpacing=0
     monospacedWidth=1
     size=17
     skewAmount=0
     typefaceFontName=Black
     outlineSettings {
      bApplyOutlineToDropShadows=false
      bMiteredCorners=false
      bSeparateFillAlpha=false
      outlineMaterial=null
      outlineSize=0
      outlineColor=<rgba 0 0 0 1>
     }
    }
   }
   slot {class=VerticalBoxSlot}
  }
 ]
}
